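# Niagara elimination effect for the RowLiaison CastSly character, emitting short-lived flickering sprite sparks from the skeletal mesh.
# One emitter splits particles into seven partitions, each with its own bone or socket and cylinder spawn shape, moved by curl noise and gravity.
# Three sprite renderers draw the particles, with flare colors blended by User.Variant and modulated by time of day and quality level.
# Hand-hidden flag collapses two partitions; foot-hidden flag targets indices beyond the seven.
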NiagaraSystem /BRCosmetics/Characters/Player/Female/Medium/Bodies/F_MED_RowLiaison_Cast/FX/NS_Character_RowLiaison_CastSly_elim
dependencies {
 usedDataInterfaces [
  &NiagaraDataInterfaceColorCurve
  &NiagaraDataInterfaceCurve
  &NiagaraDataInterfaceArrayFloat
  &NiagaraDataInterfaceSkeletalMesh
 ]
 usedDynamicInputs [
  &Multiply_Vector2D_ByFloat
  &Lerp_Float
  &MakeVector2D
 ]
 usedModules [
  &self:SystemSpawnScript.SystemScriptSource.SystemScriptGraph.NiagaraNodeAssignment_2.SetVariables_#1
  &SystemState
  /BRCosmetics/Characters/Player/Female/Medium/Bodies/F_MED_PiperShelf_Pearl/FX/TimeOfDayGlobal
 ]
}
emitters {Minimal002 {
  emitterUpdateScript [
   &EmitterState()
   &SpawnRate(SpawnRate=150, Use Spawn Probability=false)
  ]
  particleSpawnScript [
   &PartitionParticles(Number Of Partitions=7, Partition Mode:ENiagara_PartitionMode_v2=::NewEnumerator5)
   &InitializeParticle(
    Color=&LerpLinearColors(
     StartColor=<rgba 1 1 1 1>
     EndColor=&ColorFromCurve(
      ColorCurve=DataInterface() {propertyValues {
        AlphaCurve {keys [<at 0 1 RCIM_Cubic>]}
        BlueCurve {keys [
          <at 0.3001 0 RCIM_Cubic>
          {arriveTangent=2.9036, interpMode=RCIM_Cubic, leaveTangent=2.9036, time=0.3457, value=1}
          <at 0.6445 1>
          <at 0.6875 0.0015>
         ]}
        GreenCurve {keys [
          <at 0.3001 0.4696 RCIM_Cubic>
          {arriveTangent=1.5402, interpMode=RCIM_Cubic, leaveTangent=1.5402, time=0.3457, value=1}
          <at 0.6445 1>
          <at 0.6875 0.6654>
         ]}
        RedCurve {keys [<at 0.3001 1 RCIM_Cubic>,<at 0.3457 1 RCIM_Cubic>,<at 0.6445 1>,<at 0.6875 2>]}
       }}
      CurveIndex=&RandomRangeFloat(Minimum=0, Maximum=1)
     )
     LerpFactor=$User.Variant
    )
    Lifetime Max=0.6
    Lifetime Min=0.3
    Lifetime Mode=::NewEnumerator1
    Sprite Size Mode=::NewEnumerator3
    Uniform Sprite Size=&RandomRangeFloat(Minimum=15, Maximum=25)
    UsePositionOffset=false
    Write Material Random=true
   )
   &ShapeLocation(
    Cylinder Distribution=::NewEnumerator2
    Cylinder Height=&MergeHelper_FloatSwitch(0=30, 1=-30, 2=35, 3=-35, 4=-20, 5=28, 6=25, Index=$Particles.PartitionParticles.Partition, Num=::NewEnumerator5)
    Cylinder Height Midpoint=1
    Cylinder Radius=&MergeHelper_FloatSwitch(0=4, 1=4, 2=2, 3=3, 4=10, 5=10, 6=15, Index=$Particles.PartitionParticles.Partition, Num=::NewEnumerator5)
    MaskSpawnGroup=false
    Non Uniform Scale=&MergeHelper_VectorSwitch(
     0=<1 1 1>
     1=<1 1 1>
     2=<1 1 1>
     3=<1 1 1>
     4=<1 1.5 1>
     5=<1 1.5 1>
     6=<1 1 1>
     Index=$Particles.PartitionParticles.Partition
     Num=::NewEnumerator5
    )
    Offset=&MergeHelper_VectorSwitch(
     0=<5 0 0>
     1=<-5 0 0>
     2=<0 0 0>
     3=<0 0 0>
     4=<-20 -3 2>
     5=<20 3 2>
     6=<-30 0 0>
     Index=$Particles.PartitionParticles.Partition
     Num=::NewEnumerator5
    )
    Offset Mode=::NewEnumerator0
    Rotation Mode=::NewEnumerator2
    Shape Primitive=::NewEnumerator1
    UseOwnerScale=true
    V Position=&Sine(Normalized Angle=$Emitter.Age, Period=2, Scale=1, Bias=0)
    Yaw / Pitch / Roll=<0 0.25 0>
   )
   &KillParticles(Kill Particles=&Boolean_AndOperation(
     Bool1=&SetBoolByFloatComparison(
      A=0
      B=&MakeFloatFromVector(Vector=Linked() {linkedVariable {name=Particles.Position, type=NiagaraPosition}}, Channel=::NewEnumerator1)
     )
     Bool2=/Niagara/DynamicInputs/Helpers/MaskBool(
      Mask Source:NiagaraInt32=1
      Mask Parameter:NiagaraInt32=$Particles.PartitionParticles.Partition
      Mask:NiagaraInt32=8
      Masked Bool:NiagaraBool=true
      Passthrough Bool:NiagaraBool=false
     )
    ))
   &SkeletalMeshLocation(
    Bone / Socket Index=$Particles.PartitionParticles.Partition
    Bone Sampling Mode=::NewEnumerator4
    MaskSpawnGroup=false
    Orientation Sampling=::NewEnumerator1
    Sampled Position Offset:NiagaraPosition=$Particles.ShapeLocation.ShapePosition
    Skeletal Mesh=$System.Skeletal Mesh
    Velocity Sampling=::NewEnumerator0
    VelocitySpeedLimit=false
   )
  ]
  particleUpdateScript [
   &ParticleState()
   &ScaleSpriteSize(Uniform Curve Scale=&Multiply_Float(
     A=&FNTimeOfDayScalar(Night Scale=1.25, Morning Scale=1, Evening Scale=1)
     B=&Multiply_Float(
      A=&MakeCustomFloatFromBool(
       Bool=$User.bisHandHidden
       True Float=&SelectFloatFromArray(
        Direct Array Index=$Particles.PartitionParticles.Partition
        Float Selection Array=DataInterface() {propertyValues {FloatData [1,1,1,0,0,1,1,1,1]}}
       )
       False Float=1
      )
      B=&MakeCustomFloatFromBool(
       Bool=$User.bisFootHidden
       True Float=&SelectFloatFromArray(
        Direct Array Index=$Particles.PartitionParticles.Partition
        Float Selection Array=DataInterface() {propertyValues {FloatData [1,1,1,1,1,1,1,0,0]}}
       )
       False Float=1
      )
     )
    )) {module=ScaleSpriteSize001}
   &ScaleColor(
    Scale Alpha=&Multiply_Float(
     A=&FloatFromCurve(
      FloatCurve=DataInterface() {propertyValues {Curve {keys [
          {interpMode=RCIM_Cubic, tangentMode=RCTM_User, time=0, value=0}
          {arriveTangent=-2.4058, interpMode=RCIM_Cubic, leaveTangent=-2.4058, tangentMode=RCTM_User, time=0.26, value=0.8421}
          {arriveTangent=-1.1775, interpMode=RCIM_Cubic, leaveTangent=-1.1775, time=0.4613, value=0.0645}
          {arriveTangent=-0.1197, interpMode=RCIM_Cubic, leaveTangent=-0.1197, time=0.5898, value=0.4537}
          <at 1 0 RCIM_Cubic>
         ]}}}
      Scale Curve=&RandomRangeFloat(Minimum=0.1, Maximum=0.5)
     )
     B=&FloatFromCurve(
      FloatCurve=DataInterface() {propertyValues {Curve {keys [{tangentMode=RCTM_User, time=0, value=0}, {tangentMode=RCTM_User, time=1, value=1}]}}}
      Scale Curve=&FloatFromCurve(
       FloatCurve=DataInterface() {propertyValues {Curve {keys [
           {tangentMode=RCTM_User, time=0, value=0}
           {arriveTangentWeight=0.1667, interpMode=RCIM_Cubic, leaveTangentWeight=0.1667, tangentMode=RCTM_User, time=1, value=1}
          ]}}}
       CurveIndex=$Emitter.Age
      )
     )
    )
    Scale RGB=&VectorFromFloat(Value=&Multiply_Float(
      A=&RandomRangeFloat(Minimum=0.7, Maximum=0.9, Evaluation Type=::NewEnumerator1)
      B=&SelectFloatFromArray(Direct Array Index=$Engine.QualityLevel, Float Selection Array=DataInterface() {propertyValues {FloatData [10,5,1,1,1]}})
     ))
    ScaleA=true
    ScaleRGB=true
   ) {module=ScaleColor001}
   &ScaleColor(
    Scale Alpha=$System.TODAlpha
    Scale RGB=&MakeVectorFromLinearColorRGB(Color=$System.TODColor)
    ScaleA=true
    ScaleRGB=true
   )
   /SetVariables(
    Particles.CenterFlareColor:LinearColor=&MultiplyLinearColor(
     Linear Color=$Particles.Color
     Scale Factor=&MakeVector4_FromLinearColor(LinearColor=&LerpLinearColors(StartColor=<rgba 10 8.161 4.9479 1>, EndColor=<rgba 10 7.5768 3.6399 1>, LerpFactor=$User.Variant))
    )
    Particles.CenterFlareSize:Vector2f=&Multiply_Vector2(A=$Particles.SpriteSize, B=&Vector2DFromFloat(Value=0.5))
    Particles.CoronaFlareColor:LinearColor=&MultiplyLinearColor(
     Linear Color=$Particles.Color
     Scale Factor=&MakeVector4_FromLinearColor(LinearColor=&LerpLinearColors(StartColor=<rgba 0.15 0.1038 0.0382 1>, EndColor=<rgba 0.0729 0.0617 0.0153 1>, LerpFactor=$User.Variant))
    )
    Particles.CoronaFlareSize:Vector2f=&Multiply_Vector2(A=$Particles.SpriteSize, B=&Vector2DFromFloat(Value=0.75))
    Particles.FlickerFlareColor:LinearColor=&MultiplyLinearColor(
     Linear Color=$Particles.Color
     Scale Factor=&MakeVector4_FromLinearColor(LinearColor=&LerpLinearColors(StartColor=<rgba 1 0.926 0.6988 1>, EndColor=<rgba 0.0417 0.0307 0.0107 50>, LerpFactor=$User.Variant))
    )
   ) {isSetParameters=true}
   &CurlNoiseForce(Bias Noise=false, Noise Frequency=70, Noise Strength=70, Pan Noise=true, Pan Noise Field=<0 0 -0.03>)
   &GravityForce(Gravity=<0 0 -100>)
   &Drag(Drag=3, Use Linear Drag=true, Use Rotational Drag=true)
   &InheritVelocity(Inherited Velocity Amount Scale=<-0.035 -0.035 -0.035>, Speed Limit=true, UseSpeedThreshold=false)
   &SolveForcesAndVelocity(Clamp Velocity=false, Limit Acceleration=false)
   &CameraOffset(Camera Offset Amount=7) {module=CameraOffset001}
  ]
  rendererClasses [&NiagaraSpriteRendererProperties]
  renderers [
   {rendererClass=&NiagaraSpriteRendererProperties}
   {rendererClass=&NiagaraSpriteRendererProperties, rendererIndex=1}
   {rendererClass=&NiagaraSpriteRendererProperties, rendererIndex=2}
  ]
 }}
summary {
 systemName=NS_Character_RowLiaison_CastSly_elim
 emitters [{emitterName=Minimal002, rendererClasses [&NiagaraSpriteRendererProperties]}]
 userVariables [
  User.bisFootHidden {type=NiagaraBool}
  User.bisHandHidden {type=NiagaraBool}
  User.Lifetime_Scale {type=NiagaraFloat, defaultValue=0, description=""}
  User.Variant {type=NiagaraFloat, defaultValue=0, description=""}
 ]
}
