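# Animation Blueprint for the RowLiaison Cast female medium player body cosmetic.
# It copies the pose from the attached parent mesh and post-processes it through the body's Control Rig.
# Crouching, emoting, knee-sliding and DBNO-carried states drive rig inputs, along with a RowLiaison emote flag.
# Each update sets that flag by checking whether the owner skeletal mesh's anim instance is playing that emote montage.
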AnimBlueprint /BRCosmetics/Characters/Player/Female/Medium/Bodies/F_MED_RowLiaison_Cast/Meshes/F_MED_RowLiaison_Cast_AnimBP
target_skeleton=&/F_MED_RowLiaison_Cast_Skeleton
graphs [
 AnimGraph {nodes [
   n1 = Control Rig!(
    Source=Copy Pose From Mesh() {bUseAttachedParent=true}
    Alpha
    "Is Crouching ":bool=Get bIsCrouching()
    Is Emoting:bool=Get bIsPlayingEmote()
    rowliaison_Emote:bool=Get IsRowliaison_Emote() -> (IsRowliaison_Emote:bool)
    Is KneeSliding:bool=Get bIsKneeSliding()
    Is DBNOCarried:bool=Get bIsDBNOCarried()
   ) -> (Pose) {
    alphaBoolBlend {blendOption=Linear}
    controlRigAssetReference {blueprintRigClass=&/F_MED_Row_Liaison_Cast_CtrlRig.F_MED_Row_Liaison_Cast_CtrlRig_C}
   }
   Output Pose(Result=@n1) {layerGroup=None}
  ]}
 EventGraph {nodes [
   Event BlueprintUpdateAnimation()
   n1 = GetAnimInstance(self=GetOwnerSkeletalMeshComponent())
   > Is Valid(InputObject=@n1)
   > Set IsRowliaison_Emote(IsRowliaison_Emote:bool=Montage_IsPlaying(self=@n1, Montage=/BRCosmetics/Animation/Game/MainPlayer/Emotes/RowLiaison/Emote_RowLiaison_HERO_M)) -> (Output_Get:bool)
  ]}
]
variables [IsRowliaison_Emote {type=boolean}]
