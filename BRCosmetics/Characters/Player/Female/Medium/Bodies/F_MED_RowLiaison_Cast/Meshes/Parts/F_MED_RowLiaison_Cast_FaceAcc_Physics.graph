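# Physics asset for the face accessory part of the female medium RowLiaison_Cast player cosmetic, covering two earrings and a front hair strand.
# Each dyn earring and the four-link hair chain hangs from a head-linked base bone through angle-limited constraints with tight twist.
# Upper-body capsules join through free-rotation constraints; head, spine_05, right clavicle and right upper arm have sized shapes, while neck and left-side bones keep tiny capsules.
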PhysicsAsset /BRCosmetics/Characters/Player/Female/Medium/Bodies/F_MED_RowLiaison_Cast/Meshes/Parts/F_MED_RowLiaison_Cast_FaceAcc_Physics
bodies {
 Base_Earing_L [{shapeType=Capsule, length=0.505, radius=0.505, shapeName=Base_Earing_L_capsule, rotation=<rot 0 -90 0>}]
 Base_Earing_R [{shapeType=Capsule, length=0.505, radius=0.505, shapeName=Base_Earing_R_capsule, rotation=<rot 0 -90 0>}]
 Base_Front_Hair [{shapeType=Capsule, length=0.505, radius=0.505, shapeName=Base_Front_Hair_capsule, rotation=<rot 0 -90 0>}]
 clavicle_l [{shapeType=Capsule, length=0.505, radius=0.505, shapeName=clavicle_l_capsule, rotation=<rot 0 -90 0>}]
 clavicle_r [{
   shapeType=Capsule
   length=5.3499
   radius=5.224
   shapeName=clavicle_r_capsule
   center=<-2.7957 1.4826 0.7792>
   rotation=<rot 76.9164 -0.0005 -0.0005>
  }]
 dyn_Earing_L [{
   shapeType=Capsule
   length=1.7251
   radius=0.9389
   shapeName=dyn_Earing_L_capsule
   center=<1.3657 0.0002 -0.0052>
   rotation=<rot 0 -88.1139 -88.4359>
  }]
 dyn_Earing_R [{
   shapeType=Capsule
   length=1.7251
   radius=0.9388
   shapeName=dyn_Earing_R_capsule
   center=<1.3656 0.0331 -0.0002>
   rotation=<rot 0 -88.4351 -88.1146>
  }]
 dyn_Front_Hair_1 [{
   shapeType=Capsule
   length=6.6379
   radius=2.4214
   shapeName=dyn_Front_Hair_1_capsule
   center=<3.5949 -0.3033 -1.2557>
   rotation=<rot 0 87.6931 -87.1887>
  }]
 dyn_Front_Hair_2 [{
   shapeType=Capsule
   length=3.7139
   radius=1.4831
   shapeName=dyn_Front_Hair_2_capsule
   center=<4.1944 -0.1367 0.1764>
   rotation=<rot 0 91.7851 -82.4556>
  }]
 dyn_Front_Hair_3 [{
   shapeType=Capsule
   length=3.3482
   radius=1.0888
   shapeName=dyn_Front_Hair_3_capsule
   center=<2.2971 -2.0727 0.6465>
   rotation=<rot 0.4677 72.699 -87.3463>
  }]
 dyn_Front_Hair_4 [{
   shapeType=Capsule
   length=2.8043
   radius=0.852
   shapeName=dyn_Front_Hair_4_capsule
   center=<1.4898 -0.209 -1.0379>
   rotation=<rot 0 83.8546 -74.299>
  }]
 head [{shapeType=Capsule, length=4.1975, radius=6.4855, shapeName=head_capsule, center=<2.325 3.3011 -0.3934>, rotation=<rot 85.6202 179.9997 180>}]
 neck_01 [{shapeType=Capsule, length=0.505, radius=0.505, shapeName=neck_01_capsule, rotation=<rot 0 -90 0>}]
 spine_05 [{shapeType=Capsule, length=9.3226, radius=8.5273, shapeName=spine_05_capsule, center=<1.0793 -1.0176 0>, rotation=<rot 0 -90 0>}]
 upperarm_l [{shapeType=Capsule, length=0.505, radius=0.505, shapeName=upperarm_l_capsule, rotation=<rot 0 -90 0>}]
 upperarm_r [{shapeType=Capsule, length=24, radius=4, shapeName=upperarm_r_capsule, center=<-12 0.7797 1>, rotation=<rot 90 0 0>}]
}
constraints [
 {bone1Name=dyn_Earing_L, bone2Name=Base_Earing_L, swing1LimitDegrees=10, swing2LimitDegrees=25, twistLimitDegrees=1}
 {bone1Name=dyn_Earing_R, bone2Name=Base_Earing_R, swing1LimitDegrees=25, swing2LimitDegrees=10, twistLimitDegrees=1}
 {bone1Name=dyn_Front_Hair_1, bone2Name=Base_Front_Hair, swing1LimitDegrees=15, swing2LimitDegrees=15, twistLimitDegrees=1}
 {bone1Name=dyn_Front_Hair_2, bone2Name=dyn_Front_Hair_1, swing1LimitDegrees=5, swing2LimitDegrees=10, twistLimitDegrees=1}
 {bone1Name=dyn_Front_Hair_3, bone2Name=dyn_Front_Hair_2, swing1LimitDegrees=30, swing2LimitDegrees=30, twistLimitDegrees=1}
 {bone1Name=dyn_Front_Hair_4, bone2Name=dyn_Front_Hair_3, swing1LimitDegrees=30, swing2LimitDegrees=30, twistLimitDegrees=1}
 {
  bone1Name=Base_Front_Hair
  bone2Name=head
  swing1LimitDegrees=45
  swing1Motion=Free
  swing2LimitDegrees=45
  swing2Motion=Free
  twistLimitDegrees=45
  twistMotion=Free
 }
 {
  bone1Name=Base_Earing_R
  bone2Name=head
  swing1LimitDegrees=45
  swing1Motion=Free
  swing2LimitDegrees=45
  swing2Motion=Free
  twistLimitDegrees=45
  twistMotion=Free
 }
 {
  bone1Name=Base_Earing_L
  bone2Name=head
  swing1LimitDegrees=45
  swing1Motion=Free
  swing2LimitDegrees=45
  swing2Motion=Free
  twistLimitDegrees=45
  twistMotion=Free
 }
 {
  bone1Name=clavicle_l
  bone2Name=spine_05
  swing1LimitDegrees=45
  swing1Motion=Free
  swing2LimitDegrees=45
  swing2Motion=Free
  twistLimitDegrees=45
  twistMotion=Free
 }
 {
  bone1Name=clavicle_r
  bone2Name=spine_05
  swing1LimitDegrees=45
  swing1Motion=Free
  swing2LimitDegrees=45
  swing2Motion=Free
  twistLimitDegrees=45
  twistMotion=Free
 }
 {
  bone1Name=upperarm_r
  bone2Name=clavicle_r
  swing1LimitDegrees=45
  swing1Motion=Free
  swing2LimitDegrees=45
  swing2Motion=Free
  twistLimitDegrees=45
  twistMotion=Free
 }
 {
  bone1Name=upperarm_l
  bone2Name=clavicle_l
  swing1LimitDegrees=45
  swing1Motion=Free
  swing2LimitDegrees=45
  swing2Motion=Free
  twistLimitDegrees=45
  twistMotion=Free
 }
 {
  bone1Name=neck_01
  bone2Name=spine_05
  swing1LimitDegrees=45
  swing1Motion=Free
  swing2LimitDegrees=45
  swing2Motion=Free
  twistLimitDegrees=45
  twistMotion=Free
 }
 {
  bone1Name=head
  bone2Name=neck_01
  swing1LimitDegrees=45
  swing1Motion=Free
  swing2LimitDegrees=45
  swing2Motion=Free
  twistLimitDegrees=45
  twistMotion=Free
 }
]
